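; Character part modifier for the SoilBlend Haunt outfit body that animates the "eliminations state" material scalar on the part's skeletal mesh.
; After res-in it zeroes the scalar, then branches on whether it is in the front end.
; In the front end, the Loop timeline replays after repeated delays as an idle showcase.
; In a match, a watched AthenaKills stat change plays a separate timeline, which resets the scalar when finished.
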(event Custom|LoopEffect
  (bind (_direction _fadein _eliminations_state) (Timeline|Loop PlayFromStart)
    (:Update
      (Rendering|Material|SetScalarParameterValueonMaterials (CustomCharacter|GetSkeletalMeshForPartOnPlayerPawn) "eliminations state" _eliminations_state))
    (:Finished
      (Utilities|FlowControl|Delay 7.0)
      (CallFunction|LoopEffect))))

(event EventOnPostResIn
  (bind _returnvalue (CustomCharacter|GetSkeletalMeshForPartOnPlayerPawn))
  (|CallParentFunction)
  (if (CustomCosmeticModifier|IsinFrontEnd)
    (Rendering|Material|SetScalarParameterValueonMaterials (CustomCharacter|GetSkeletalMeshForPartOnPlayerPawn) "eliminations state" 0.0)
    (Utilities|FlowControl|Delay 7.0)
    (Timeline|Loop PlayFromStart)
    (else
      (Rendering|Material|SetScalarParameterValueonMaterials _returnvalue "eliminations state" 0.0)
      (Cosmetics|WatchCosmeticStat (CustomCharacter|GetAssociatedPlayerPawn) "(TagName=\"StatManager.AthenaKills\")" 0 false
        (:OnStatChanged
          (bind (_direction _fadein _eliminations_state) (Timeline|Timeline_0 PlayFromStart)
            (:Update
              (Rendering|Material|SetScalarParameterValueonMaterials _returnvalue "eliminations state" _eliminations_state))
            (:Finished
              (Rendering|Material|SetScalarParameterValueonMaterials _returnvalue "eliminations state" 0.0))))))))

(fn ConstructionScript)
