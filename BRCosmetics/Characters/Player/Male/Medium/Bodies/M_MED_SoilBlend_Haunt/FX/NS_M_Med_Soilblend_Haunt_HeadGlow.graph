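# Niagara head glow effect for the SoilBlend Haunt male character cosmetic, combining a sprite and a particle light.
# A burst spawns an offset particle that persists until the emitter deactivates, with a platform-dependent camera offset.
# A variant selector picks cyan, red or green, and that color also feeds the light.
# User.bisHeadHidden switches the light off and sets the particle visibility tag, so the glow follows head visibility.
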NiagaraSystem /BRCosmetics/Characters/Player/Male/Medium/Bodies/M_MED_SoilBlend_Haunt/FX/NS_M_Med_Soilblend_Haunt_HeadGlow
dependencies {
 usedDataInterfaces [&NiagaraDataInterfacePlatformSet]
 usedDynamicInputs []
 usedModules [&SystemState, &FN_VariantManager]
}
emitters {HeadGlow {
  emitterUpdateScript [
   &EmitterState(Life Cycle Mode=::NewEnumerator1, Loop Behavior=::NewEnumerator1, Loop Duration Mode=::NewEnumerator1, UseLoopDelay=false)
   &SpawnBurst_Instantaneous(Use Spawn Probability=false, UseLoopCountLimit=false)
  ]
  particleSpawnScript [
   &InitializeParticle(Position Offset=<7 0 0>, Sprite Size Mode=::NewEnumerator3, Uniform Sprite Size=55, UsePositionOffset=true, Write Material Random=true)
   &CameraOffset(Camera Offset Amount=&FN_PickFloatDependOnPlatform(TargetPlatform=DataInterface() {propertyValues {Platforms {qualityLevelMask=-29}}}, TargetPlatformActive=-10, TargetPlatformNoActive=0))
  ]
  particleUpdateScript [
   &ParticleState(Kill Particles When Lifetime Has Elapsed=false, Let Infinitely Lived Particles Die When Emitter Deactivates=true)
   &Color(Color=&VariantColorSelector(Variant Amount=1, Original Color=<rgba 0.094 0.8702 1 1>, Variant 1 Color=<rgba 1 0 0.0394 1>, Variant 2 Color=<rgba 0.2977 1 0.3602 1>))
   &FN_Light_Attributes(
    Color=$Particles.Color
    Create Color Attribute=true
    Create Exponent Attribute=true
    Create Radius Attribute=true
    Enable Particle Light Enable=true
    Exponent=2
    Light Enabled:NiagaraBool=&Boolean_NotOperation(Bool=$User.bisHeadHidden)
    Radius=45
    Scale Color=3
   )
   /SetVariables(Particles.VisibilityTag=&MakeCustomIntFromBool(Bool=$User.bisHeadHidden, True INT=1, False INT=0)) {isSetParameters=true}
  ]
  rendererClasses [&NiagaraSpriteRendererProperties, &NiagaraLightRendererProperties]
  renderers [
   {rendererClass=&NiagaraSpriteRendererProperties}
   {rendererClass=&NiagaraLightRendererProperties, rendererIndex=1}
  ]
 }}
summary {
 systemName=NS_M_Med_Soilblend_Haunt_HeadGlow
 emitters [{emitterName=HeadGlow, rendererClasses [&NiagaraSpriteRendererProperties, &NiagaraLightRendererProperties]}]
 userVariables [
  User.bisHeadHidden {type=NiagaraBool}
  User.Variant {type=NiagaraFloat, defaultValue=2, description=""}
 ]
}
