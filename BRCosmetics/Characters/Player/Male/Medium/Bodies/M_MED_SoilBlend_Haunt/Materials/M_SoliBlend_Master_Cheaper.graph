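# Master character material for the SoilBlend Haunt body, a lit, glass-like shell with animated internal smoke and a selectable skull face.
# Custom HLSL ray-marches scrolling cloud noise for depth, with shading-path switches for mobile.
# The skull uses virtual-plane UVs, parallax occlusion mapping, noise distortion and fading masks.
# Fresnel, matcap-style fake refraction and color blends build the surface and glow, wrapped by shared character PreFX/PostFX functions.
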Material /BRCosmetics/Characters/Player/Male/Medium/Bodies/M_MED_SoilBlend_Haunt/Materials/M_SoliBlend_Master_Cheaper
graph {
 properties {
  bUseMaterialAttributes=true
  bUsedWithMeshDeformer=true
  bUsedWithMorphTargets=true
  bUsedWithSkeletalMesh=true
  pixelDepthOffsetMode=PDOM_Legacy
  usedShadingModels=Default Lit
 }
 nodes [
  n1 = Constant() {r=0.05}
  n2 = ScalarParameter() {defaultValue=0.35, group=SoilHaunt, parameterName=Specular}
  n3 = NamedRerouteUsage()
  n4 = Constant() {r=0.3}
  n5 = Constant() {r=1.7}
  n6 = NamedRerouteUsage()
  n7 = LinearInterpolate(
   A=Multiply(A=NamedRerouteUsage(), B=LinearInterpolate(A=@n4, B=@n5, Alpha=@n6) {constB=1.35}) {constB=0.1}
   B=NamedRerouteUsage()
   Alpha=NamedRerouteUsage()
  )
  n8 = Constant() {r=0.75}
  n9 = NamedRerouteUsage()
  n10 = MaterialFunctionCall() {materialFunction=&MF_CharacterParent_2023_PreFX}
  MaterialFunctionCall(
   Material=MakeMaterialAttributes!(
    BaseColor=NamedRerouteUsage()
    Metallic=ScalarParameter() {defaultValue=0.1, group=SoilHaunt, parameterName=Metallic}
    Specular=LinearInterpolate(A=@n1, B=@n2, Alpha=@n3)
    Roughness=NamedRerouteUsage()
    Anisotropy
    EmissiveColor=MaterialFunctionCall(Emissive=@n7) {materialFunction=&MF_FX_EmissiveAdjustments}.Result
    Opacity
    OpacityMask
    Normal
    Tangent
    WorldPositionOffset
    SubsurfaceColor
    ClearCoat
    ClearCoatRoughness
    AmbientOcclusion=LinearInterpolate(A=@n8, Alpha=@n9)
    Refraction
    CustomizedUVs_0
    CustomizedUVs_1
    CustomizedUVs_2
    CustomizedUVs_3
    CustomizedUVs_4
    CustomizedUVs_5
    CustomizedUVs_6
    CustomizedUVs_7
    PixelDepthOffset
    ShadingModel
    Displacement
   )
   M=@n10.M
   Pre Skinned Local Position=@n10.M
   Distance Blend=@n10.M
   Fresnel=@n10.M
  ) {materialFunction=&MF_CharacterParent_2023_PostFX}
  n11 = TextureSampleParameter2D(UVs=NamedRerouteUsage()) {
   group=SoilHaunt
   parameterName=MaskDepth_Style01
   samplerSource=SSM_Clamp_WorldGroupSettings
   samplerType=SAMPLERTYPE_Masks
   texture=/BRCosmetics/Characters/Player/Male/Medium/Bodies/M_MED_SoilBlend_Haunt/Textures/T_skull_face_Depth
  }
  n12 = OneMinus(None=ComponentMask(None=@n11.RGB) {g=true})
  NamedRerouteDeclaration(None=@n12) {name=Emissive Mask, nodeColor=<rgba 0 1 0.6016 1>}
  n13 = NamedRerouteUsage()
  n14 = Multiply(A=NamedRerouteUsage(), B=ScalarParameter() {defaultValue=0.08, parameterName=Distortion weight})
  n15 = MaterialFunctionCall(Vector or UVs=Add(A=NamedRerouteUsage()) {constB=-0.5}) {materialFunction=&VectorToRadialValue}
  n16 = Panner(Coordinate=@n15.Radial Coordinates) {speedX=0.002, speedY=-0.05}
  n17 = TextureCoordinate()
  n18 = DDX(Value=@n17)
  n19 = DDY(Value=@n17)
  n20 = TextureSampleParameter2D(UVs=NamedRerouteUsage()) {
   group=SoilHaunt
   parameterName=MaskDepth_Style01
   samplerSource=SSM_Clamp_WorldGroupSettings
   samplerType=SAMPLERTYPE_Masks
   texture=&T_skull_face_Depth
  }
  n21 = Multiply(
   A=LinearInterpolate(
    A=TextureSample!(UVs=@n16, Tex, "DDX(UVs)"=@n18, "DDY(UVs)"=@n19, Apply View MipBias) {mipValueMode=TMVM_Derivative, texture=&AtmosphericCloudNoise01}.R
    Alpha=NamedRerouteUsage()
   ) {constB=0}
   B=LinearInterpolate(A=Constant() {r=0.2}, B=Constant() {r=0.02}, Alpha=@n20.B) {constB=0.01}
  ) {constB=0.1}
  NamedRerouteDeclaration(None=Add(A=NamedRerouteDeclaration(None=@n13) {name=Final_parallax UV, nodeColor=<rgba 0.0781 0 1 1>}, B=Add(A=@n14, B=@n21))) {name=Parallax anim UV, nodeColor=<rgba 1 0 0.3516 1>}
  n22 = ScalarParameter() {defaultValue=0.2, group=SoilHaunt, parameterName=Fresnel Inner Transparency}
  n23 = ScalarParameter() {defaultValue=0.7, group=SoilHaunt, parameterName=Fresnel Outer Transparency}
  n24 = Fresnel(ExponentIn=Constant() {r=1}, Normal=VertexNormalWS())
  n25 = ScalarParameter() {defaultValue=2, group=SoilHaunt, parameterName=Glass thickness}
  n26 = NamedRerouteDeclaration(None=Clamp(None=Multiply(A=@n24, B=@n25))) {name=Fresnel, nodeColor=<rgba 0.1016 1 0 1>}
  n27 = OneMinus(None=LinearInterpolate(A=Constant() {r=0.75}, B=LinearInterpolate(A=@n22, B=@n23, Alpha=@n26), Alpha=NamedRerouteUsage()))
  n28 = VectorParameter() {group=SoilHaunt, parameterName=Bright_Color, defaultValue=<rgba 0.1468 1 0.7473 1>}
  n29 = ScalarParameter() {defaultValue=1.9, group=SoilHaunt, parameterName=Internal luminance multiple}
  n30 = Constant()
  n31 = MaterialFunctionCall(Brightness=Constant() {r=1.1}, Exponent=Constant() {r=10}) {materialFunction=&MF_Fresnel_FX}
  n32 = TextureSampleParameter2D(UVs=NamedRerouteUsage()) {
   group=SoilHaunt
   parameterName=MaskColor_style02
   samplerSource=SSM_Clamp_WorldGroupSettings
   texture=/BRCosmetics/Characters/Player/Male/Medium/Bodies/M_MED_SoilBlend_Haunt/Textures/T_skull_face_Color_01
  }
  n33 = Desaturation(None=@n32.RGB, Fraction=Constant() {r=1})
  n34 = Constant() {r=0.75}
  n35 = Saturate(None=NamedRerouteUsage())
  n36 = ScalarParameter() {defaultValue=0.85, group=SoilHaunt, parameterName=height_basecolor_exponent}
  n37 = NamedRerouteUsage()
  n38 = ScalarParameter() {defaultValue=0.35, parameterName=TopSmokeStyle01}
  n39 = ScalarParameter() {defaultValue=1.5, parameterName=base smoke concentration}
  n40 = ScalarParameter() {defaultValue=4, parameterName=Pbase smoke concentration Mobile}
  n41 = Power(Base=OneMinus(None=ComponentMask(None=@n11.RGB) {r=true}), Exp=Constant() {r=2.5}) {constExponent=1.5}
  n42 = Add(
   A=Multiply(A=NamedRerouteUsage(), B=Constant() {r=1}) {constB=2}
   B=LinearInterpolate(A=LinearInterpolate(A=@n38, Alpha=@n37), B=ShadingPathSwitch(Default=@n39, Mobile=@n40), Alpha=Saturate(None=@n41))
  ) {constB=0.35}
  n43 = LinearInterpolate(
   A=VectorParameter() {group=SoilHaunt, parameterName=Dark_color, defaultValue=<rgba 0.0067 0.09 0.0792 1>}.RGB
   B=LinearInterpolate(A=Multiply(A=@n28.RGB) {constB=0.5}, B=Multiply(A=@n28.RGB, B=@n29) {constB=10}, Alpha=LinearInterpolate(A=@n30, Alpha=@n31.Fresnel))
   Alpha=LinearInterpolate(A=Multiply(A=@n33, B=@n34) {constB=0.5}, B=Power(Base=@n35, Exp=@n36), Alpha=Saturate(None=@n42))
  )
  NamedRerouteDeclaration(None=Multiply(A=Multiply(A=@n27, B=@n43), B=Constant() {r=0.75})) {name=BaseColor, nodeColor=<rgba 0 1 0.5547 1>}
  MaterialExpressionMultiply_13 = Multiply(
   A=TextureCoordinate()
   B=MaterialFunctionCall(X=Constant() {r=1}, Y=Constant() {r=1.1}) {materialFunction=&MakeFloat2}.Result
  )
  MaterialExpressionTransform_0 = Transform(None=CameraVectorWS()) {transformSourceType=TRANSFORMSOURCE_World, transformType=TRANSFORM_Tangent}
  MaterialExpressionConstant_27 = Constant() {r=1}
  MaterialExpressionConstant_34 = Constant() {r=0.006}
  MaterialExpressionTextureObject_3 = TextureObject() {texture=&AtmosphericCloudNoise01}
  MaterialExpressionVertexNormalWS_1 = VertexNormalWS()
  MaterialExpressionTime_0 = Time()
  n44 = VectorParameter() {parameterName=heightspeed, defaultValue=<rgba -0.02 0 0 1>}
  n45 = ScalarParameter() {defaultValue=1, parameterName=Velocity multiple}
  MaterialExpressionReroute_0 = Reroute(None=ComponentMask(None=Multiply(A=@n44.RGB, B=@n45)) {g=true, r=true})
  MaterialExpressionComponentMask_10 = ComponentMask(None=Multiply(A=@n45, B=Constant3Vector() {constant=<rgba -0.04 0 0 1>})) {g=true, r=true}
  MaterialExpressionConstant_36 = Constant() {r=0.06}
  MaterialExpressionDivide_0 = Divide(B=@MaterialExpressionConstant_27)
  n46 = Custom!(
   uv=@MaterialExpressionMultiply_13
   viewDir=@MaterialExpressionTransform_0
   steps=@MaterialExpressionConstant_27
   stepdist=@MaterialExpressionConstant_34
   tex=@MaterialExpressionTextureObject_3
   vertnormal=@MaterialExpressionVertexNormalWS_1
   dTex=@MaterialExpressionTextureObject_3
   Time=@MaterialExpressionTime_0
   heightspeed=@MaterialExpressionReroute_0
   distortspeed=@MaterialExpressionComponentMask_10
   distortionstrength=@MaterialExpressionConstant_36
   StepSize=@MaterialExpressionDivide_0
  ) {
   code="float3 rayStep = viewDir * -1;\r\nfloat2 uv2 = uv + Time*heightspeed;\r\nuv += Time*distortspeed;\r\nfloat4 heightTex = Texture2DSample(tex,texSampler,uv);\r\nfloat4 distortionTex = Texture2DSample(dTex,texSampler,uv2);\r\n\r\nfor(int i=0;i<=int(steps);i++)\r\n{\r\nif(heightTex.r<=i/steps)\r\n{\r\nreturn float((heightTex.r*i)/16.0)*1.0; \r\n}\r\nfloat stepct = heightTex.r * stepdist;\r\nuv += rayStep * stepct;\r\nuv2 += rayStep * stepct;\r\ndistortionTex = Texture2DSample(dTex,texSampler,uv2*1.0);\r\nfloat2 distortStr = distortionTex.xy * distortionstrength;\r\nheightTex = Texture2DSample(tex,texSampler,uv.xy + distortStr);\r\nRayMarchUV = uv.xy + distortStr;\r\n}\r\n\r\nreturn heightTex;"
   description=RayMarchTexture2
   outputType=CMOT_Float1
   showCode=true
   additionalOutputs [{outputName=RayMarchUV, outputType=CMOT_Float2}]
   inputs [
    {
     inputName=uv
     input {expression=&self:MaterialExpressionMultiply_13}
    }
    {
     inputName=viewDir
     input {expression=&self:MaterialExpressionTransform_0}
    }
    {
     inputName=steps
     input {expression=&self:MaterialExpressionConstant_27}
    }
    {
     inputName=stepdist
     input {expression=&self:MaterialExpressionConstant_34}
    }
    {
     inputName=tex
     input {expression=&self:MaterialExpressionTextureObject_3}
    }
    {
     inputName=vertnormal
     input {expression=&self:MaterialExpressionVertexNormalWS_1}
    }
    {
     inputName=dTex
     input {expression=&self:MaterialExpressionTextureObject_3}
    }
    {
     inputName=Time
     input {expression=&self:MaterialExpressionTime_0}
    }
    {
     inputName=heightspeed
     input {expression=&self:MaterialExpressionReroute_0}
    }
    {
     inputName=distortspeed
     input {expression=&self:MaterialExpressionComponentMask_10}
    }
    {
     inputName=distortionstrength
     input {expression=&self:MaterialExpressionConstant_36}
    }
    {
     inputName=StepSize
     input {expression=&self:MaterialExpressionDivide_0}
    }
   ]
  }
  n47 = LinearInterpolate(B=Constant() {r=5}, Alpha=@n46.return) {constB=3}
  n48 = Constant() {r=0.02}
  NamedRerouteDeclaration(None=Saturate(None=MaterialFunctionCall(In=@n47, Contrast=@n48) {materialFunction=&CheapContrast}.Result)) {name=MobileRayMarch, nodeColor=<rgba 0 1 0.2266 1>}
  n49 = Add(
   A=VectorParameter() {group=RhombCamoVisor, parameterName=VP_XAxis, defaultValue=<rgba 1 0 0 1>}.RGB
   B=VectorParameter() {group=RhombCamoVisor, parameterName=VP_PlaneXAxis_Offset, defaultValue=<rgba 0 0 0 1>}.RGB
  )
  n50 = Add(
   A=VectorParameter() {group=RhombCamoVisor, parameterName=VP_PlaneCenter, defaultValue=<rgba -1 7.5 157 1>}.RGB
   B=Multiply(A=ScalarParameter() {defaultValue=-9, group=SoilHaunt, parameterName=CenterOffset}, B=NamedRerouteUsage())
  )
  n51 = ScalarParameter() {defaultValue=27, group=SoilHaunt, parameterName=VP_Scale_style01}
  n52 = ScalarParameter() {defaultValue=25, group=SoilHaunt, parameterName=VP_Scale_style01_Mobile}
  n53 = NamedRerouteDeclaration(None=Add(
    A=VectorParameter() {group=SoilHaunt, parameterName=VP_PlaneNormal, defaultValue=<rgba 0 -1 0 1>}.RGB
    B=VectorParameter() {group=SoilHaunt, parameterName=VP_PlaneNormal_Offset, defaultValue=<rgba 0 0 0 1>}.RGB
   )) {name=VP_PlaneNormal_WS, nodeColor=<rgba 1 0.3516 0 1>}
  NamedRerouteDeclaration(None=MaterialFunctionCall(
    Plane Normal Axis=@n53
    Plane X Axis=NamedRerouteDeclaration(None=@n49) {name=VP_PlaneXAxis_WS, nodeColor=<rgba 0.4062 0 1 1>}
    Plane Center=NamedRerouteDeclaration(None=@n50) {name=VP_PlaneCenter_WS, nodeColor=<rgba 1 0 0.5391 1>}
    UV Scale=ShadingPathSwitch(Default=@n51, Mobile=@n52)
   ) {materialFunction=&VirtualPlaneCoordinates}.UVs) {name=VirtualPlaneUV, nodeColor=<rgba 1 0 0.9375 1>}
  n54 = TextureSampleParameter2D(UVs=Panner(Coordinate=TextureCoordinate() {uTiling=0.3, vTiling=0.3}) {speedX=0.04}) {
   group=SoilHaunt
   parameterName=AtmosphericCloudNoise
   samplerSource=SSM_Wrap_WorldGroupSettings
   texture=&AtmosphericCloudNoise01
  }
  n55 = Sine(None=Multiply(A=Time()) {constB=0.15})
  NamedRerouteDeclaration(None=Multiply(A=Multiply(A=@n54.R) {constB=1.05}, B=Saturate(None=@n55))) {name=Eliminations, nodeColor=<rgba 1 0 0.8906 1>}
  n56 = Constant() {r=2}
  n57 = Constant() {r=2}
  n58 = Multiply(
   A=TextureCoordinate() {coordinateIndex=1}
   B=ShadingPathSwitch(
    Default=MaterialFunctionCall(X=@n56, Y=@n57) {materialFunction=&MakeFloat2}.Result
    Mobile=Constant() {r=1.5}
   )
  ) {constB=2}
  n59 = TextureSampleParameter2D(UVs=Panner(Coordinate=@n58) {speedX=0.04}) {
   group=SoilHaunt
   parameterName=AtmosphericCloudNoise
   samplerSource=SSM_Wrap_WorldGroupSettings
   texture=&AtmosphericCloudNoise01
  }
  n60 = TextureSampleParameter2D(UVs=Add(A=Panner(Coordinate=@n58) {speedX=0.02}, B=Multiply(A=@n59.R) {constB=0.08})) {
   group=SoilHaunt
   parameterName=AtmosphericCloudNoise
   samplerSource=SSM_Wrap_WorldGroupSettings
   texture=&AtmosphericCloudNoise01
  }
  NamedRerouteDeclaration(None=@n60.R) {name=Surface smoke, nodeColor=<rgba 1 0 0.5391 1>}
  n61 = NamedRerouteUsage()
  MaterialExpressionMultiply_4 = Multiply(
   A=TextureCoordinate()
   B=MaterialFunctionCall(X=Constant() {r=1}, Y=ScalarParameter() {defaultValue=1.1, group=SoilHaunt, parameterName=texcoord_UVY_Tile}) {materialFunction=&MakeFloat2}.Result
  )
  MaterialExpressionTransform_2 = Transform(None=CameraVectorWS()) {transformSourceType=TRANSFORMSOURCE_World, transformType=TRANSFORM_Tangent}
  MaterialExpressionConstant_24 = Constant() {r=16}
  MaterialExpressionScalarParameter_24 = ScalarParameter() {defaultValue=0.0105, group=SoilHaunt, parameterName=step_dist}
  MaterialExpressionTextureObject_1 = TextureObject() {texture=&AtmosphericCloudNoise01}
  MaterialExpressionVertexNormalWS_0 = VertexNormalWS()
  MaterialExpressionTime_4 = Time()
  n62 = VectorParameter() {parameterName=heightspeed, defaultValue=<rgba -0.02 0 0 1>}
  n63 = ScalarParameter() {defaultValue=1, parameterName=Velocity multiple}
  MaterialExpressionReroute_3 = Reroute(None=ComponentMask(None=Multiply(A=@n62.RGB, B=@n63)) {g=true, r=true})
  MaterialExpressionComponentMask_3 = ComponentMask(None=Multiply(A=@n63, B=VectorParameter() {parameterName=distortspeed, defaultValue=<rgba -0.04 0 0 1>}.RGB)) {g=true, r=true}
  MaterialExpressionScalarParameter_28 = ScalarParameter() {defaultValue=0.1, group=SoilHaunt, parameterName=distortion_strength}
  MaterialExpressionDivide_3 = Divide(B=@MaterialExpressionConstant_24)
  n64 = Custom(
   uv=@MaterialExpressionMultiply_4
   viewDir=@MaterialExpressionTransform_2
   steps=@MaterialExpressionConstant_24
   stepdist=@MaterialExpressionScalarParameter_24
   tex=@MaterialExpressionTextureObject_1
   vertnormal=@MaterialExpressionVertexNormalWS_0
   dTex=@MaterialExpressionTextureObject_1
   Time=@MaterialExpressionTime_4
   heightspeed=@MaterialExpressionReroute_3
   distortspeed=@MaterialExpressionComponentMask_3
   distortionstrength=@MaterialExpressionScalarParameter_28
   StepSize=@MaterialExpressionDivide_3
  ) {
   code="float3 rayStep = viewDir * -1;\r\nfloat2 uv2 = uv + Time*heightspeed;\r\nuv += Time*distortspeed;\r\nfloat4 heightTex = Texture2DSample(tex,texSampler,uv);\r\nfloat4 distortionTex = Texture2DSample(dTex,texSampler,uv2);\r\n\r\nfor(int i=0;i<=int(steps);i++)\r\n{\r\nif(heightTex.r<=i/steps)\r\n{\r\nreturn float((heightTex.r*i)/16.0)*1.0; \r\n}\r\nfloat stepct = heightTex.r * stepdist;\r\nuv += rayStep * stepct;\r\nuv2 += rayStep * stepct;\r\ndistortionTex = Texture2DSample(dTex,texSampler,uv2*1.0);\r\nfloat2 distortStr = distortionTex.xy * distortionstrength;\r\nheightTex = Texture2DSample(tex,texSampler,uv.xy + distortStr);\r\nRayMarchUV = uv.xy + distortStr;\r\n}\r\n\r\nreturn heightTex;"
   description=RayMarchTexture2
   outputType=CMOT_Float1
   showCode=true
   additionalOutputs [{outputName=RayMarchUV, outputType=CMOT_Float2}]
   inputs [
    {
     inputName=uv
     input {expression=&self:MaterialExpressionMultiply_4}
    }
    {
     inputName=viewDir
     input {expression=&self:MaterialExpressionTransform_2}
    }
    {
     inputName=steps
     input {expression=&self:MaterialExpressionConstant_24}
    }
    {
     inputName=stepdist
     input {expression=&self:MaterialExpressionScalarParameter_24}
    }
    {
     inputName=tex
     input {expression=&self:MaterialExpressionTextureObject_1}
    }
    {
     inputName=vertnormal
     input {expression=&self:MaterialExpressionVertexNormalWS_0}
    }
    {
     inputName=dTex
     input {expression=&self:MaterialExpressionTextureObject_1}
    }
    {
     inputName=Time
     input {expression=&self:MaterialExpressionTime_4}
    }
    {
     inputName=heightspeed
     input {expression=&self:MaterialExpressionReroute_3}
    }
    {
     inputName=distortspeed
     input {expression=&self:MaterialExpressionComponentMask_3}
    }
    {
     inputName=distortionstrength
     input {expression=&self:MaterialExpressionScalarParameter_28}
    }
    {
     inputName=StepSize
     input {expression=&self:MaterialExpressionDivide_3}
    }
   ]
  }
  n65 = NamedRerouteUsage()
  n66 = TextureSample!(UVs, Tex, Apply View MipBias) {
   samplerSource=SSM_Wrap_WorldGroupSettings
   samplerType=SAMPLERTYPE_LinearColor
   texture=/BRCosmetics/Characters/Player/Male/Medium/Bodies/M_MED_SoilBlend_Haunt/Textures/T_soilblend_mask
  }
  n67 = Multiply(A=@n66.G) {constB=1.2}
  NamedRerouteDeclaration(None=LinearInterpolate(A=ShadingPathSwitch(Default=@n64.return, Mobile=@n61), B=Power(Base=@n65), Alpha=Saturate(None=@n67))) {name=RayMarchColor_output, nodeColor=<rgba 0.2422 0 1 1>}
  n68 = TransformPosition(None=WorldPosition().XYZ) {transformSourceType=TRANSFORMPOSSOURCE_World, transformType=TRANSFORMPOSSOURCE_View}
  n69 = VertexNormalWS()
  n70 = MaterialFunctionCall(Color=CrossProduct(A=Normalize(VectorInput=@n68), B=Transform(None=@n69) {transformSourceType=TRANSFORMSOURCE_World, transformType=TRANSFORM_View})) {materialFunction=&SplitComponents}
  n71 = Multiply(
   A=MaterialFunctionCall(R=@n70.G, G=@n70.G) {materialFunction=&AppendMany}.RG
   B=Constant2Vector() {g=0.5, r=0.5}
  )
  NamedRerouteDeclaration(None=ComponentMask(None=Add(A=@n71) {constB=0.5}) {g=true, r=true}) {name=MatCapUV, nodeColor=<rgba 1 0 0.2578 1>}
  n72 = MaterialFunctionCall(UVs=NamedRerouteUsage(), Texture Scale=Constant() {r=1.13}) {materialFunction=&ScaleUVsByCenter}
  n73 = Multiply(A=Multiply(
    A=TextureSample(UVs=@n72.UVs) {texture=/BRCosmetics/Characters/Player/Male/Medium/Bodies/M_MED_SoilBlend_Haunt/Textures/T_Fake_refraction}.R
    B=VectorParameter() {group=SoilHaunt, parameterName=Bright_Color, defaultValue=<rgba 0.1468 1 0.7473 1>}.RGB
   )) {constB=0.15}
  n74 = QualitySwitch(Default=Constant())
  NamedRerouteDeclaration(None=Add(A=ShadingPathSwitch(Default=@n73, Mobile=@n74), B=NamedRerouteUsage())) {name=Emissive Color, nodeColor=<rgba 1 0 0.1875 1>}
  n75 = Multiply(A=MaterialFunctionCall(Brightness=Constant() {r=3}, Exponent=Constant() {r=2}) {materialFunction=&MF_Fresnel_FX}.Fresnel)
  NamedRerouteDeclaration(None=Power(Base=Saturate(None=@n75))) {name=fresnel2, nodeColor=<rgba 0 0.2734 1 1>}
  NamedRerouteDeclaration(None=ScalarParameter() {desc="0 no face, 1 skull face ", group=SoilHaunt, parameterName=Face_style_Select, sliderMax=1, sortPriority=0}) {name=Global_blend, nodeColor=<rgba 1 0.0469 0 1>}
  n76 = NamedRerouteUsage()
  n77 = CameraPositionWS()
  n78 = DotProduct(A=Normalize(VectorInput=Subtract(A=@n76, B=@n77)), B=Normalize(VectorInput=@n53))
  n79 = Constant() {r=1.5}
  n80 = Saturate(None=OneMinus(None=Power(Base=@n78, Exp=@n79) {constExponent=1.5}))
  n81 = NamedRerouteUsage()
  n82 = NamedRerouteUsage()
  n83 = LinearInterpolate(
   A=LinearInterpolate(B=NamedRerouteDeclaration(None=@n80) {name=VirtualPlaneMask, nodeColor=<rgba 1 0.1875 0 1>}, Alpha=Saturate(None=@n81)) {constA=1}
   Alpha=Saturate(None=Subtract(A=@n82) {constB=3})
  )
  n84 = TextureSampleParameter2D(UVs=Panner(Coordinate=TextureCoordinate() {uTiling=0.7, vTiling=0.7}) {speedX=-0.02}) {
   group=SoilHaunt
   parameterName=AtmosphericCloudNoise
   samplerSource=SSM_Wrap_WorldGroupSettings
   texture=&AtmosphericCloudNoise01
  }
  n85 = Saturate(None=LinearInterpolate(
    A=ScalarParameter() {defaultValue=0.25, group=SoilHaunt, parameterName=Selective_fade_out, sortPriority=0}
    B=ScalarParameter() {defaultValue=1.2, group=SoilHaunt, parameterName=Selective_fade_in, sortPriority=0}
    Alpha=@n84.R
   ))
  NamedRerouteDeclaration(None=LinearInterpolate(A=LinearInterpolate(B=@n83, Alpha=@n85) {constA=1}, B=@n83, Alpha=ScalarParameter() {group=SoilHaunt, parameterName=eliminations state})) {name=SkullSphereMask, nodeColor=<rgba 1 0.4688 0 1>}
  n86 = TextureObject() {samplerType=SAMPLERTYPE_Masks, texture=&T_skull_face_Depth}
  n87 = Constant() {r=12}
  NamedRerouteDeclaration(None=MaterialFunctionCall!(
    Heightmap Texture=@n86
    Height Ratio=ScalarParameter() {defaultValue=0.25, parameterName=HeightRatio_Style01}
    Min Steps=@n87
    Max Steps=@n87
    UVs=NamedRerouteUsage()
    Heightmap Channel=Constant4Vector() {constant=<rgba 0 1 0 0>}.RGBA
    Reference Plane=Constant() {r=1}
    -----------------------------
    Use World Coordinates
    Specify Manual Texture Size
    Manual Texture Size
    ------------------------------
    "Render Shadows (Occlusion Mapping)"=StaticBool()
    Light Vector
    Shadow Steps
    Shadow Penumbra
    Transform To VertexNormal
   ) {materialFunction=/Engine/Functions/Engine_MaterialFunctions01/Texturing/ParallaxOcclusionMapping}.Parallax UVs) {name=Parallax UV_style01, nodeColor=<rgba 0.5938 0 1 1>}
  NamedRerouteDeclaration(None=@n12) {name=SkullRounghness, nodeColor=<rgba 0 1 0.9297 1>}
  NamedRerouteDeclaration(None=SphereMask(A=@n15.Linear Distance, B=Constant(), Radius=Constant() {r=0.45}) {attenuationRadius=0.35, hardnessPercent=0}) {name=VP_Spheremask, nodeColor=<rgba 1 0 0.4687 1>}
  n88 = Constant() {r=0.1}
  n89 = Constant() {r=0.35}
  n90 = NamedRerouteUsage()
  n91 = LinearInterpolate(A=Constant() {r=0.1}, B=Constant() {r=0.05}, Alpha=NamedRerouteUsage())
  n92 = TextureSample(UVs=Multiply(A=TextureCoordinate()) {constB=4}) {texture=&T_Fake_refraction}
  NamedRerouteDeclaration(None=LinearInterpolate(
    A=LinearInterpolate(A=@n88, B=@n89, Alpha=@n90) {constA=0.7}
    B=LinearInterpolate(A=@n91, Alpha=@n92.G) {constB=0.5}
    Alpha=NamedRerouteUsage()
   )) {name=Rounghness, nodeColor=<rgba 0.0547 0 1 1>}
  VertexInterpolator(VS=PreSkinnedPosition())
 ]
}
